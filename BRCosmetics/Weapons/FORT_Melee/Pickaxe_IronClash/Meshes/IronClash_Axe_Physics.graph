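# Physics asset for the IronClash pickaxe axe mesh, giving capsule bodies to the two-handed handle bone and two dyn_ bones, arm and wrist.
# The bodies form a jointed chain from handle to arm to wrist.
# The handle-to-arm joint locks twist and one swing axis, leaving a single hinge-like swing axis.
# The arm-to-wrist joint has no rotation axis explicitly locked and carries a very narrow twist limit value.
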PhysicsAsset /BRCosmetics/Weapons/FORT_Melee/Pickaxe_IronClash/Meshes/IronClash_Axe_Physics
bodies {
 Melee_TwoHanded_Handle [{shapeType=Capsule, length=15, radius=4.0809, shapeName=Melee_TwoHanded_Handle_capsule}]
 dyn_arm [{shapeType=Capsule, length=15, radius=6.4028, shapeName=dyn_arm_capsule, center=<7.7471 0 0>, rotation=<rot 90 179.9999 -179.9999>}]
 dyn_wrist [{shapeType=Capsule, length=15, radius=6.4028, shapeName=dyn_wrist_capsule, center=<7.7471 0 0>, rotation=<rot 90 -176.2074 -176.2073>}]
}
constraints [
 {
  bone1Name=dyn_arm
  bone2Name=Melee_TwoHanded_Handle
  swing1LimitDegrees=45
  swing1Motion=Locked
  swing2LimitDegrees=30
  twistLimitDegrees=45
  twistMotion=Locked
 }
 {bone1Name=dyn_wrist, bone2Name=dyn_arm, swing1LimitDegrees=45, swing2LimitDegrees=45, twistLimitDegrees=1}
]
